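# Control Rig for the TrackWaver tail accessory that poses the tail by player movement state.
# Setup appends a tip bone to the C_Tail_A joint chain and caches it and the dyn_skinTail bones as arrays.
# Forward solve clamps the first two joints' bend, then layers offsets weighted by velocity, parachuting, skydive-dive pitch and knee sliding.
# A Hermite spline through the posed joints then drives the denser skin chain via stretched spline fitting.
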ControlRigBlueprint /BRCosmetics/Accessories/FORT_Tails/Tail_F_MED_TrackWaver_Yes/Meshes/F_MED_TrackWaver_Yes_Tail_CtrlRig
graphs [
 RigVMFunctionLibrary {links [], nodes []}
 RigVMFunction_Sequence_2_ContainedGraph {nodes [
   n1 = RigVMFunctionEntryNode()
   RigVMFunction_Sequence = &RigVMFunction_Sequence(ExecuteContext=@n1.ExecuteContext)
   n2 = &RigVMFunction_Sequence(ExecuteContext=@RigVMFunction_Sequence.B)
   n3 = &RigVMFunction_Sequence(ExecuteContext=@n2.B)
   n4 = &RigVMFunction_Sequence(ExecuteContext=@n3.B)
   RigVMFunctionReturnNode(A=@RigVMFunction_Sequence.A, B=@n2.A, C=@n3.A, D=@n4.A, E=@n4.B)
  ]}
 RigVMModel {nodes [
   n1 = &RigUnit_PrepareForExecution()
   n2 = &RigVMFunction_Sequence(ExecuteContext=@n1.ExecutePin)
   n3 = RigVMDispatchNode(
    Index=0
    Array=/Script/ControlRig.RigUnit_HierarchyGetParentsItemArray(Child=(Type=Bone,Name="C_Tail_A_9_Jnt"), bDefaultParent=false, bIncludeChild=true, bReverse=true) {is=struct RigVMUnitNode}.Parents
    ExecuteContext=@n2.A
   ) {template="DISPATCH_RigVMDispatch_ArrayRemove(io Array,in Index)"}
   > n4 = /Script/ControlRig.RigUnit_HierarchyAddBone(
    Name=C_Tail_A_10_Jnt
    Parent=(Type=Bone,Name="C_Tail_A_9_Jnt")
    Space=LocalSpace
    Transform=(Rotation=(X=0,Y=0,Z=0,W=1),Translation=(X=8.878,Y=0,Z=0),Scale3D=(X=1,Y=1,Z=1))
   ) {is=struct RigVMUnitNode}
   > n5 = RigVMDispatchNode(Other=((Type=None,Name="None")), Other.0=@n4.Item, Array=@n3.Array) {template="DISPATCH_RigVMDispatch_ArrayAppend(io Array,in Other)"}
   > RigVMVariableNode(Value=@n5.Array) {variable=Tail Orig}
   Get Children by Contained Strings = RigVMFunctionReferenceNode(
    Include Item=true
    Item=(Type=Bone,Name="dyn_skinTail_master")
    Recursive=true
    Search String=skinTail
    Type to Search=Bone
    ExecuteContext=@n2.B
   )
   > RigVMVariableNode(Value=@"Get Children by Contained Strings".Results) {variable=Tail New}
   RigUnit_BeginExecution = &RigUnit_BeginExecution()
   RigVMFunction_Sequence = RigVMAggregateNode(ExecuteContext=@RigUnit_BeginExecution.ExecutePin)
   n6 = &RigUnit_GetTransform(Item=(Type=Bone,Name="C_Tail_A_1_Jnt"), Space=LocalSpace, bInitial=False)
   RigVMFunction_MathQuaternionToEuler = &RigVMFunction_MathQuaternionToEuler(RotationOrder=ZYX, Value=@n6.Transform.Rotation)
   Clamp = &RigVMFunction_MathDoubleClamp(Maximum=25, Value=@RigVMFunction_MathQuaternionToEuler.Result.Y)
   RigVMFunction_MathQuaternionFromEuler = &RigVMFunction_MathQuaternionFromEuler(
    Euler=(X=0,Y=0,Z=0)
    Euler.Y=@Clamp.Result
    Euler.X=@RigVMFunction_MathQuaternionToEuler.Result.X
    Euler.Z=@RigVMFunction_MathQuaternionToEuler.Result.Z
   )
   RigUnit_SetTransform = &RigUnit_SetTransform(
    Item=(Type=Bone,Name="C_Tail_A_1_Jnt")
    Space=LocalSpace
    Value=(Rotation=(X=0,Y=0,Z=0,W=1),Translation=(X=0,Y=0,Z=0),Scale3D=(X=1,Y=1,Z=1))
    Weight=1
    bInitial=False
    bPropagateToChildren=true
    Value.Scale3D=@n6.Transform.Scale3D
    Value.Translation=@n6.Transform.Translation
    Value.Rotation=@RigVMFunction_MathQuaternionFromEuler.Result
    ExecutePin=@RigVMFunction_Sequence.A
   )
   n7 = &RigUnit_GetTransform(Item=(Type=Bone,Name="C_Tail_A_2_Jnt"), Space=LocalSpace, bInitial=False)
   n8 = &RigVMFunction_MathQuaternionToEuler(RotationOrder=ZYX, Value=@n7.Transform.Rotation)
   > &RigUnit_SetTransform(
    Item=(Type=Bone,Name="C_Tail_A_2_Jnt")
    Space=LocalSpace
    Value=(Rotation=(X=0,Y=0,Z=0,W=1),Translation=(X=0,Y=0,Z=0),Scale3D=(X=1,Y=1,Z=1))
    Weight=1
    bInitial=False
    bPropagateToChildren=true
    Value.Scale3D=@n7.Transform.Scale3D
    Value.Translation=@n7.Transform.Translation
    Value.Rotation=&RigVMFunction_MathQuaternionFromEuler(
     Euler=(X=0,Y=0,Z=0)
     Euler.Y=&RigVMFunction_MathDoubleClamp(Maximum=25, Value=@n8.Result.Y).Result
     Euler.X=@n8.Result.X
     Euler.Z=@n8.Result.Z
    ).Result
   )
   Remap = &RigVMFunction_MathDoubleRemap(SourceMaximum=-65, SourceMinimum=0, TargetMaximum=1, TargetMinimum=0, Value=RigVMVariableNode() {variable=TailDivePitch}.Value)
   RigVMFunction_MathBoolNot = &RigVMFunction_MathBoolNot(Value=RigVMVariableNode() {variable=isParachuting}.Value)
   RigVMFunction_MathBoolAnd = &RigVMFunction_MathBoolAnd(A=@RigVMFunction_MathBoolNot.Result, B=RigVMVariableNode() {variable=isSkydiveDive}.Value)
   n9 = &RigVMFunction_AlphaInterp(
    InterpSpeedDecreasing=10
    InterpSpeedIncreasing=10
    OutRange=(Min=1,Max=0)
    Scale=1
    bClampResult=false
    bMapRange=false
    Value=RigVMDispatchNode(False=0, True=@Remap.Result, Condition=@RigVMFunction_MathBoolAnd.Result) {template="DISPATCH_RigVMDispatch_If(in Condition,in True,in False,out Result)"}.Result
   )
   &RigUnit_ModifyTransforms(
    ItemToModify="((Item=(Type=Bone,Name=\"C_Tail_A_1_Jnt\"),Transform=(Rotation=(X=0,Y=-0.1305,Z=0,W=0.9914),Translation=(X=0,Y=0,Z=0),Scale3D=(X=1,Y=1,Z=1))),(Item=(Type=Bone,Name=\"C_Tail_A_2_Jnt\"),Transform=(Rotation=(X=0,Y=-0.1305,Z=0,W=0.9914),Translation=(X=0,Y=0,Z=0),S ...(truncated)"
    WeightMinimum=0
    Weight=@n9.Result
    ExecutePin=@RigVMFunction_Sequence.C
   )
   RigVMFunction_MathBoolToFloat = &RigVMFunction_MathBoolToFloat(Value=RigVMVariableNode() {variable=isKneeSliding}.Value)
   n10 = &RigVMFunction_AlphaInterp(InterpSpeedDecreasing=10, InterpSpeedIncreasing=10, Scale=1, bClampResult=False, Value=@RigVMFunction_MathBoolToFloat.Result)
   &RigUnit_ModifyTransforms(
    ItemToModify=((Item=(Type=Bone,Name="C_Tail_A_1_Jnt"),Transform=(Rotation=(X=0,Y=-0.342,Z=0,W=0.9397),Translation=(X=0,Y=0,Z=0),Scale3D=(X=1,Y=1,Z=1))))
    WeightMinimum=0
    Weight=@n10.Result
    ExecutePin=@RigVMFunction_Sequence.D
   )
   SplineFromItems = RigVMFunctionReferenceNode(Samples Per Segment=16, Spline Mode=Hermite, ExecuteContext=@RigVMFunction_Sequence.E, Items=RigVMVariableNode() {variable=Tail Orig}.Value)
   > /Script/ControlRigSpline.RigUnit_FitChainToSplineCurveItemArray(
    Alignment=Stretched
    DebugSettings=(bEnabled=False,Scale=1,CurveColor=(R=1,G=1,B=0,A=1),SegmentsColor=(R=1,G=0,B=0,A=1),WorldOffset=(Rotation=(X=0,Y=0,Z=0,W=1),Translation=(X=0,Y=0,Z=0),Scale3D=(X=1,Y=1,Z=1)))
    Maximum=1
    Minimum=0
    PoleVectorPosition=(X=0,Y=0,Z=0)
    PrimaryAxis=(X=1,Y=0,Z=0)
    RotationEaseType=Linear
    Rotations=()
    SamplingPrecision=12
    SecondaryAxis=(X=0,Y=0,Z=0)
    Weight=1
    bPropagateToChildren=True
    Spline=@SplineFromItems.Spline
    Items=RigVMVariableNode() {variable=Tail New}.Value
   ) {is=struct RigVMUnitNode}
   RigUnit_GetTransform = &RigUnit_GetTransform(Item=(Type=Bone,Name="dyn_skinTail_15"), Space=LocalSpace, bInitial=False)
   > &RigUnit_SetTransform(
    Item=(Type=Bone,Name="dyn_skinTail_15")
    Space=LocalSpace
    Value=(Rotation=(X=0,Y=0,Z=0,W=1),Translation=(X=5.15,Y=0,Z=0),Scale3D=(X=1,Y=1,Z=1))
    Weight=1
    bInitial=False
    Value.Rotation=@RigUnit_GetTransform.Transform.Rotation
    Value.Scale3D=@RigUnit_GetTransform.Transform.Scale3D
    Value.Translation.Y=@RigUnit_GetTransform.Transform.Translation.Y
    Value.Translation.Z=@RigUnit_GetTransform.Transform.Translation.Z
   )
   &RigUnit_ModifyTransforms(
    ItemToModify=((Item=(Type=Bone,Name="C_Tail_A_Base_Jnt"),Transform=(Rotation=(X=0,Y=0,Z=0.2588,W=0.9659),Translation=(X=0,Y=0,Z=0),Scale3D=(X=1,Y=1,Z=1))))
    Weight=1
    WeightMinimum=0
    ExecutePin=@RigVMFunction_Sequence.B
   )
   AlphaInterp = &RigVMFunction_AlphaInterp(
    InRange=(Min=0,Max=770)
    InterpSpeedDecreasing=15
    InterpSpeedIncreasing=15
    OutRange=(Min=1,Max=0)
    Scale=1
    bClampResult=true
    bMapRange=true
    Value=RigVMVariableNode() {variable=abs Velocity}.Value
   )
   > &RigUnit_ModifyTransforms(
    ItemToModify="((Item=(Type=Bone,Name=\"C_Tail_A_5_Jnt\"),Transform=(Rotation=(X=0,Y=0,Z=0,W=1),Translation=(X=0,Y=0,Z=0),Scale3D=(X=1,Y=1,Z=1))),(Item=(Type=Bone,Name=\"C_Tail_A_6_Jnt\"),Transform=(Rotation=(X=0,Y=-0.0872,Z=0,W=0.9962),Translation=(X=0,Y=0,Z=0),Scale3D=(X=1 ...(truncated)"
    WeightMinimum=0
    Weight=@AlphaInterp.Result
    ItemToModify.4.Transform.Rotation=/Script/RigVM.RigVMFunction_MathQuaternionFromRotatorV2(Value=RigVMDispatchNode(Elements=(Pitch=40,Yaw=0,Roll=0)) {template="DISPATCH_RigVMDispatch_MakeStruct(in Elements,out Struct)"}.Struct) {is=struct RigVMUnitNode, template="Cast::Execute(in Value,out Result)"}.Result
   )
   > &RigUnit_ModifyTransforms(
    ItemToModify="((Item=(Type=Bone,Name=\"C_Tail_A_1_Jnt\"),Transform=(Rotation=(X=0,Y=-0.2164,Z=0,W=0.9763),Translation=(X=0,Y=0,Z=0),Scale3D=(X=1,Y=1,Z=1))),(Item=(Type=Bone,Name=\"C_Tail_A_2_Jnt\"),Transform=(Rotation=(X=0,Y=0,Z=0,W=1),Translation=(X=0,Y=0,Z=0),Scale3D=(X=1 ...(truncated)"
    WeightMinimum=0
    Weight=RigVMDispatchNode(True=0.5, Condition=RigVMVariableNode() {variable=isParachuting}.Value, False=@AlphaInterp.Result) {template="DISPATCH_RigVMDispatch_If(in Condition,in True,in False,out Result)"}.Result
   )
   &RigUnit_ModifyTransforms(
    ExecutePin=()
    ItemToModify=((Item=(Type=Bone,Name="dyn_skinTail_15"),Transform=(Rotation=(X=0,Y=0,Z=0,W=1),Translation=(X=0,Y=0,Z=0),Scale3D=(X=1,Y=1,Z=1))))
    Weight=1
    WeightMinimum=0
   )
  ]}
]
hierarchy [
 C_Root_Main_Root_Jnt {loc=<0 0 0>, rot=<0 0 0 1>}
 C_Tail_A_Base_Jnt {parent=C_Root_Main_Root_Jnt, loc=<1.4997 -0.3348 0.4223>, rot=<-0.5 0.5 0.5 -0.5>}
 C_Tail_A_1_Jnt {parent=C_Tail_A_Base_Jnt, loc=<8.8758 -0.0429 0.1469>, rot=<0.7071 0.0063 0.0115 -0.707>}
 C_Tail_A_2_Jnt {parent=C_Tail_A_1_Jnt, loc=<8.8775 0.0376 0.011>, rot=<0 0.0009 -0.0031 -1>}
 C_Tail_A_3_Jnt {parent=C_Tail_A_2_Jnt, loc=<8.879 0.0015 0.0004>, rot=<0 0 -0.0001 -1>}
 C_Tail_A_4_Jnt {parent=C_Tail_A_3_Jnt, loc=<8.8778 0 0>, rot=<0 0 0 1>}
 C_Tail_A_5_Jnt {parent=C_Tail_A_4_Jnt, loc=<8.8778 0 0>, rot=<0 0 0 1>}
 C_Tail_A_6_Jnt {parent=C_Tail_A_5_Jnt, loc=<8.8775 0 0>, rot=<0 0 0 1>}
 C_Tail_A_7_Jnt {parent=C_Tail_A_6_Jnt, loc=<8.8776 0 0>, rot=<0 0 0 1>}
 C_Tail_A_8_Jnt {parent=C_Tail_A_7_Jnt, loc=<8.8788 0 0>, rot=<0 0 0 1>}
 C_Tail_A_9_Jnt {parent=C_Tail_A_8_Jnt, loc=<8.8781 0 0>, rot=<0 0 0 1>}
 dyn_skinTail_master {parent=C_Root_Main_Root_Jnt, loc=<1.4997 -0.3348 0.4223>, rot=<-0.5 0.5 0.5 -0.5>}
 dyn_skinTail_01 {parent=dyn_skinTail_master, loc=<5.15 -0.0429 0.1469>, rot=<0.7071 0.0063 0.0115 -0.707>}
 dyn_skinTail_02 {parent=dyn_skinTail_01, loc=<5.15 0.0376 0.011>, rot=<0 0.0009 -0.0031 -1>}
 dyn_skinTail_03 {parent=dyn_skinTail_02, loc=<5.15 0.0015 0.0004>, rot=<0 0 -0.0001 -1>}
 dyn_skinTail_04 {parent=dyn_skinTail_03, loc=<5.15 0 0>, rot=<0 0 0 1>}
 dyn_skinTail_05 {parent=dyn_skinTail_04, loc=<5.15 0 0>, rot=<0 0 0 1>}
 dyn_skinTail_06 {parent=dyn_skinTail_05, loc=<5.15 0 0>, rot=<0 0 0 1>}
 dyn_skinTail_07 {parent=dyn_skinTail_06, loc=<5.15 0 0>, rot=<0 0 0 1>}
 dyn_skinTail_08 {parent=dyn_skinTail_07, loc=<5.15 0 0>, rot=<0 0 0 1>}
 dyn_skinTail_09 {parent=dyn_skinTail_08, loc=<5.15 0 0>, rot=<0 0 0 1>}
 dyn_skinTail_10 {parent=dyn_skinTail_09, loc=<5.15 0 0>, rot=<0 0 0 1>}
 dyn_skinTail_11 {parent=dyn_skinTail_10, loc=<5.15 0 0>, rot=<0 0 0 1>}
 dyn_skinTail_12 {parent=dyn_skinTail_11, loc=<5.15 0 0>, rot=<0 0 0 1>}
 dyn_skinTail_13 {parent=dyn_skinTail_12, loc=<5.15 0 0>, rot=<0 0 0 1>}
 dyn_skinTail_14 {parent=dyn_skinTail_13, loc=<5.15 0 0>, rot=<0 0 0 1>}
 dyn_skinTail_15 {parent=dyn_skinTail_14, loc=<5.15 0 0>, rot=<0 0 0 1>}
]
variables [
 Tail New {type="TArray<FRigElementKey>"}
 Tail Orig {type="TArray<FRigElementKey>"}
 Tail Pose Alpha {type=double, default=0}
 TailDivePitch {type=double, default=0}
 abs Velocity {type=double, default=0}
 isKneeSliding {type=bool, default=False}
 isParachuting {type=bool, default=False}
 isSkydiveDive {type=bool, default=False}
]
